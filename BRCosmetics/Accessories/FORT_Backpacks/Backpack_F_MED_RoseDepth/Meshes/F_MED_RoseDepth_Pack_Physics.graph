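# Physics asset for the RoseDepth backpack cosmetic, a puppet-like rig with head, arms and skirt.
# Small capsules on the base bones join the puppet_master body, which joins pack, through constraints with free swing and twist.
# Capsule chains of dyn bones extend from those bases to the head, both hands and skirt.
# Tight angular limits restrict these chains, with the skirt allowed the widest swing and its twist locked.
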PhysicsAsset /BRCosmetics/Accessories/FORT_Backpacks/Backpack_F_MED_RoseDepth/Meshes/F_MED_RoseDepth_Pack_Physics
bodies {
 base_neck [{shapeType=Capsule, length=0.505, radius=0.505, shapeName=base_neck_capsule, rotation=<rot 0 -90 0>}]
 base_shoulder_l [{shapeType=Capsule, length=0.505, radius=0.505, shapeName=base_shoulder_l_capsule, rotation=<rot 0 -90 0>}]
 base_shoulder_r [{shapeType=Capsule, length=0.505, radius=0.505, shapeName=base_shoulder_r_capsule, rotation=<rot 0 -90 0>}]
 base_skirt_00 [{shapeType=Capsule, length=0.505, radius=0.505, shapeName=base_skirt_00_capsule, rotation=<rot 0 -90 0>}]
 dyn_elbow_l [{shapeType=Capsule, length=2, radius=2, shapeName=dyn_elbow_l_capsule, center=<3.3301 0 0>, rotation=<rot 90 0 0>}]
 dyn_elbow_r [{shapeType=Capsule, length=2, radius=2, shapeName=dyn_elbow_r_capsule, center=<-4.0194 0 0>, rotation=<rot 90 0 0>}]
 dyn_hand_l [{shapeType=Capsule, length=2, radius=2, shapeName=dyn_hand_l_capsule, center=<3.3559 0 0>, rotation=<rot 90 0 0>}]
 dyn_hand_r [{shapeType=Capsule, length=2, radius=2, shapeName=dyn_hand_r_capsule, center=<-3.8129 0 0>, rotation=<rot 90 0 0>}]
 dyn_head [{shapeType=Capsule, length=2, radius=1, shapeName=dyn_head_capsule, center=<2 0 0>, rotation=<rot 90 0 0>}]
 dyn_neck [{shapeType=Capsule, length=2, radius=1, shapeName=dyn_neck_capsule, center=<2 0 0>, rotation=<rot 90 0 0>}]
 dyn_shoulder_l [{shapeType=Capsule, length=1.5, radius=1.5, shapeName=dyn_shoulder_l_capsule, center=<2 0 0>, rotation=<rot 90 0 0>}]
 dyn_shoulder_r [{shapeType=Capsule, length=2, radius=1, shapeName=dyn_shoulder_r_capsule, center=<-2 0 0>, rotation=<rot 90 0 0>}]
 dyn_skirt_00 [{shapeType=Capsule, length=5, radius=3, shapeName=dyn_skirt_00_capsule, center=<8 0 0>, rotation=<rot 90 0 0>}]
 pack [{shapeType=Capsule, length=0.505, radius=0.505, shapeName=pack_capsule, rotation=<rot 0 -90 0>}]
 puppet_master [{
   shapeType=Capsule
   length=5.1959
   radius=4.3707
   shapeName=puppet_master_capsule
   center=<2.281 0.4173 3.5635>
   rotation=<rot -7.2452 -90 1.8034>
  }]
}
constraints [
 {
  bone1Name=base_neck
  bone2Name=puppet_master
  swing1LimitDegrees=45
  swing1Motion=Free
  swing2LimitDegrees=45
  swing2Motion=Free
  twistLimitDegrees=45
  twistMotion=Free
 }
 {bone1Name=dyn_neck, bone2Name=base_neck, swing1LimitDegrees=5, swing2LimitDegrees=1, twistLimitDegrees=1}
 {bone1Name=dyn_head, bone2Name=dyn_neck, swing1LimitDegrees=5, swing2LimitDegrees=5, twistLimitDegrees=1}
 {
  bone1Name=base_shoulder_l
  bone2Name=puppet_master
  swing1LimitDegrees=45
  swing1Motion=Free
  swing2LimitDegrees=45
  swing2Motion=Free
  twistLimitDegrees=45
  twistMotion=Free
 }
 {bone1Name=dyn_shoulder_l, bone2Name=base_shoulder_l, swing1LimitDegrees=7, swing2LimitDegrees=7, twistLimitDegrees=1}
 {bone1Name=dyn_elbow_l, bone2Name=dyn_shoulder_l, swing1LimitDegrees=0.1, swing2LimitDegrees=4, twistLimitDegrees=1}
 {bone1Name=dyn_hand_l, bone2Name=dyn_elbow_l, swing1LimitDegrees=0.5, swing2LimitDegrees=1, twistLimitDegrees=1}
 {
  bone1Name=base_shoulder_r
  bone2Name=puppet_master
  swing1LimitDegrees=45
  swing1Motion=Free
  swing2LimitDegrees=45
  swing2Motion=Free
  twistLimitDegrees=45
  twistMotion=Free
 }
 {bone1Name=dyn_shoulder_r, bone2Name=base_shoulder_r, swing1LimitDegrees=4, swing2LimitDegrees=4, twistLimitDegrees=1}
 {bone1Name=dyn_elbow_r, bone2Name=dyn_shoulder_r, swing1LimitDegrees=0.1, swing2LimitDegrees=2, twistLimitDegrees=1}
 {bone1Name=dyn_hand_r, bone2Name=dyn_elbow_r, swing1LimitDegrees=0.5, swing2LimitDegrees=1, twistLimitDegrees=1}
 {
  bone1Name=base_skirt_00
  bone2Name=puppet_master
  swing1LimitDegrees=45
  swing1Motion=Free
  swing2LimitDegrees=45
  swing2Motion=Free
  twistLimitDegrees=45
  twistMotion=Free
 }
 {bone1Name=dyn_skirt_00, bone2Name=base_skirt_00, swing1LimitDegrees=17, swing2LimitDegrees=15, twistLimitDegrees=1, twistMotion=Locked}
 {
  bone1Name=puppet_master
  bone2Name=pack
  swing1LimitDegrees=45
  swing1Motion=Free
  swing2LimitDegrees=45
  swing2Motion=Free
  twistLimitDegrees=45
  twistMotion=Free
 }
]
