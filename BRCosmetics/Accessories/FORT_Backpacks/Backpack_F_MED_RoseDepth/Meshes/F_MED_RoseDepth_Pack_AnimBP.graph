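# Anim Blueprint for the RoseDepth backpack cosmetic, adding secondary physics motion to the pack.
# Two chained AnimDynamics on lower_puppet (one front-end only, one weaker while skydiving outside it) and a puppet_master offset feed a cached pose.
# That pose drives two base-bone-space RigidBody simulations with separate physics assets, blended by bIsInFrontEnd.
# The update event computes skydiving alpha and a gravity override via a deprecated state function.
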AnimBlueprint /BRCosmetics/Accessories/FORT_Backpacks/Backpack_F_MED_RoseDepth/Meshes/F_MED_RoseDepth_Pack_AnimBP
target_skeleton=&/F_MED_RoseDepth_Pack_Skeleton
graphs [
 AnimGraph {nodes [
   n1 = Get GravityOverride()
   n2 = RigidBody!(
    ComponentPose=Local To Component(LocalPose=Use cached pose 'Inputpose'())
    bAlphaBoolEnabled:bool
    OverrideWorldGravity:Vector=@n1
    ExternalForce
   ) -> (Pose) {
    alphaInputType=Bool
    bOverrideWorldGravity=true
    overridePhysicsAsset=&/F_MED_RoseDepth_Pack_Physics_FE
    simulationSpace=BaseBoneSpace
    alphaBoolBlend {blendOption=Linear}
    baseBoneRef {boneName=pack}
    componentAppliedLinearAccClamp=<1000 1000 1000>
    componentLinearAccScale=<0.1 0.1 0.1>
    componentLinearVelScale=<0.1 0.1 0.1>
    simSpaceSettings {
     maxAngularAcceleration=500
     maxAngularVelocity=3
     maxLinearAcceleration=100
     maxLinearVelocity=100
     velocityScaleZ=0
     worldAlpha=0.1
     externalLinearDragV=<0.2 0.2>
    }
   }
   n3 = RigidBody!(ComponentPose=Local To Component(LocalPose=Use cached pose 'Inputpose'()), Alpha:float, OverrideWorldGravity=@n1, ExternalForce) -> (Pose) {
    bOverrideWorldGravity=true
    overridePhysicsAsset=&/F_MED_RoseDepth_Pack_Physics
    simulationSpace=BaseBoneSpace
    alphaBoolBlend {blendOption=Linear}
    baseBoneRef {boneName=pack}
    componentAppliedLinearAccClamp=<3000 3000 4000>
    componentLinearAccScale=<0.3 0.2 0.4>
    componentLinearVelScale=<0.3 0.2 0.4>
    simSpaceSettings {
     maxAngularAcceleration=1000
     maxAngularVelocity=3
     maxLinearAcceleration=700
     maxLinearVelocity=500
     velocityScaleZ=0
     worldAlpha=0.2
     externalLinearDragV=<0.2 0.2>
    }
   }
   Output Pose(Result=Blend Poses by bool(BlendPose_0=Component To Local(ComponentPose=@n2), BlendPose_1=Component To Local(ComponentPose=@n3), bActiveValue=Get bIsInFrontEnd()) {blendType=Linear}) {layerGroup=None}
   n4 = Anim Dynamics - Bone: lower_puppet!(
    ComponentPose=Local To Component(LocalPose=Input Pose())
    bAlphaBoolEnabled:bool=Get bIsInFrontEnd()
    GravityOverride:Vector="0,0,0"
    ExternalForce="0,0,0"
   ) -> (Pose) {
    alphaInputType=Bool
    angularDampingOverride=1
    angularSpringConstant=10000000
    bAngularSpring=true
    bOverrideAngularDamping=true
    bUseGravityOverride=true
    gravityScale=0.35
    simulationSpace=BoneRelative
    alphaBoolBlend {blendOption=Linear}
    alphaScaleBiasClamp {bInterpResult=true}
    boundBone {boneName=lower_puppet}
    componentAppliedLinearAccClamp=<0.1 0.1 0.1>
    componentLinearAccScale=<0.1 0.1 0.1>
    componentLinearVelScale=<0.1 0.1 0.1>
    gravityOverride {z=-980}
    physicsBodyDefinitions [{
      boundBone {boneName=lower_puppet}
      boxExtents=<50 50 50>
      constraintSetup {angularLimitsMax=<15 15 0>, angularLimitsMin=<-15 -15 -15>, angularTarget=<1 0 0>}
      localJointOffset=<0 0 0>
     }]
    relativeSpaceBone {boneName=pack}
   }
   n5 = Anim Dynamics - Bone: lower_puppet!(ComponentPose=@n4, Alpha:float=Get SkyDiving_Alpha() -> (SkyDiving_Alpha:double), GravityOverride=@n1, ExternalForce="0,0,0") -> (Pose) {
    angularDampingOverride=1
    angularSpringConstant=10000
    bAngularSpring=true
    bOverrideAngularDamping=true
    bUseGravityOverride=true
    gravityScale=0.35
    simulationSpace=BoneRelative
    alphaBoolBlend {blendOption=Linear}
    alphaScaleBiasClamp {bInterpResult=true}
    boundBone {boneName=lower_puppet}
    componentLinearAccScale=<0.3 0.2 0.4>
    componentLinearVelScale=<0.3 0.2 0.4>
    gravityOverride {z=-980}
    physicsBodyDefinitions [{
      boundBone {boneName=lower_puppet}
      boxExtents=<20 20 20>
      constraintSetup {angularLimitsMax=<15 45 0>, angularLimitsMin=<-15 -45 -45>, angularTarget=<1 0 0>}
      localJointOffset=<5 0 0>
     }]
    relativeSpaceBone {boneName=pack}
   }
   n6 = "Transform (Modify) Bone - Bone: puppet_master"(ComponentPose=@n5, Alpha=1, Translation="-1.8,0,0", Rotation="0,0,0", Scale="1,1,1") {
    translationMode=BMM_Additive
    translationSpace=BCS_BoneSpace
    alphaBoolBlend {blendOption=Linear}
    boneToModify {boneName=puppet_master}
    translation {x=-1.8}
   }
   Save cached pose 'Inputpose'(Pose=Component To Local(ComponentPose=@n6)) {cacheName=Inputpose}
  ]}
 EventGraph {nodes [
   Event BlueprintUpdateAnimation()
   > n1 = Sequence()
   > n2 = Gravity Override State Function DEPRECATED(CharacterPartAnimInstance=Self-Reference(), GameTimeInSeconds=GetGameTimeInSeconds(), GravityOverrideParams=Get Gravity Override Params())
   > Set GravityOverride(GravityOverride=@n2)
   n3 = NOT Boolean(A=Get bIsInFrontEnd())
   n4 = Get bIsSkyDiving()
   Set SkyDiving_Alpha(execute=@n1.then_1, SkyDiving_Alpha:double=SelectFloat(A=0.2, B=0.7, bPickA=AND Boolean(A=@n4, B=@n3))) -> (Output_Get:double)
   TryGetPawnOwner()
  ]}
]
variables [
 Gravity Override Params {type=GravityOverrideParamsStruct}
 GravityOverride {type=Vector}
 SkyDiving_Alpha {type=number}
]
